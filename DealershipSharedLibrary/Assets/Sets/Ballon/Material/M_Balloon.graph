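# Subsurface material for balloons, tinted by particle color with particle alpha as the dithered opacity mask.
# An RGB mask texture places the tint, Fresnel-driven plastic bag layer and Fresnel subsurface glow.
# UseCreepyBalloonMovement blends two vertex-offset setups: curl-noise sway weighted by height and vertex color, and layered-sine rope wiggle along UV V.
# Low quality and mobile feature levels zero out subsurface color and opacity.
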Material /DealershipSharedLibrary/Assets/Sets/Ballon/Material/M_Balloon
graph {
 properties {
  bUseMaterialAttributes=true
  bUsedWithMeshDeformer=true
  bUsedWithSkeletalMesh=true
  ditherOpacityMask=true
  pixelDepthOffsetMode=PDOM_Legacy
  shadingModel=MSM_Subsurface
  usedShadingModels=Subsurface
 }
 nodes [
  n1 = MaterialFunctionCall(Float3=Subtract(
    A=WorldPosition().XYZ
    B=MaterialFunctionCall() {materialFunction=&ObjectPivotPoint}.Object Pivot Location
   )) {materialFunction=&BreakOutFloat3Components}
  n2 = Add(A=@n1.B, B=ScalarParameter() {defaultValue=12, parameterName=AdjsutPivot})
  n3 = Time()
  n4 = Sine(None=Multiply(A=@n3))
  n5 = Sine(None=Multiply(A=Add(A=@n3) {constB=0.25}))
  n6 = Sine(None=Multiply(A=Add(A=@n3) {constB=0.5}))
  n7 = VectorNoise(World Position=Add(
    A=ObjectPositionWS()
    B=MaterialFunctionCall(X=@n4, Y=@n5, Z=@n6) {materialFunction=&MakeFloat3}.Result
   )) {noiseFunction=VNF_CurlALU}
  n8 = ComponentMask(None=Subtract(
    A=WorldPosition().XYZ
    B=MaterialFunctionCall() {materialFunction=&ObjectPivotPoint}.Object Pivot Location
   )) {b=true}
  n9 = MaterialFunctionCall() {materialFunction=&ObjectLocalBounds}
  n10 = OneMinus(None=Divide(A=Add(A=@n8), B=ComponentMask(None=@n9.Local Bounds Size) {b=true}))
  n11 = MaterialFunctionCall(Float3=Divide(A=WorldPosition().XYZ) {constB=128}) {materialFunction=&BreakOutFloat3Components}
  n12 = Time()
  n13 = Add(A=@n11.R, B=Multiply(A=@n12))
  n14 = Add(A=@n11.G, B=Multiply(A=Add(A=@n12) {constB=0.25}))
  n15 = Add(A=@n11.B, B=Multiply(A=Add(A=@n12) {constB=0.75}))
  n16 = Multiply(A=Add(
    A=ObjectPositionWS()
    B=MaterialFunctionCall(X=@n13, Y=@n14, Z=@n15) {materialFunction=&MakeFloat3}.Result
   ))
  n17 = MaterialFunctionCall(Float3=Add(A=@n7, B=VectorNoise(World Position=@n16) {noiseFunction=VNF_CurlALU})) {materialFunction=&BreakOutFloat3Components}
  n18 = Multiply(
   A=Multiply(A=Clamp(None=@n10), B=VertexColor().R)
   B=MaterialFunctionCall(X=@n17.R, Y=@n17.R, Z=Clamp(None=@n17.B)) {materialFunction=&MakeFloat3}.Result
  )
  n19 = Add(
   A=Normalize(VectorInput=MaterialFunctionCall(X=@n1.R, Y=@n1.R, Z=@n2) {materialFunction=&MakeFloat3}.Result)
   B=Add(A=Multiply(A=@n7), B=Multiply(A=@n18))
  )
  n20 = ComponentMask(None=@n19) {b=true}
  NamedRerouteDeclaration(None=AppendVector(A=ComponentMask(None=@n19) {g=true, r=true}, B=Clamp(None=@n20) {maxDefault=1000})) {name=MF_Balloon Movement, nodeColor=<rgba 0.2656 1 0 1>}
  n21 = Multiply(
   A=ComponentMask(None=TextureCoordinate()) {g=true}
   B=MaterialFunctionCall() {materialFunction=/Engine/Functions/Engine_MaterialFunctions02/WorldPositionOffset/ObjectScale}.Scale X
  )
  n22 = ScalarParameter() {defaultValue=1, parameterName=TimeScale}
  n23 = Multiply(A=Time(), B=Multiply(A=Constant() {r=0.1}, B=@n22)) {constB=0.1}
  n24 = Multiply(A=Sine(None=Add(A=@n21, B=@n23)) {period=0.5}, B=Constant3Vector() {constant=<rgba 0 0 0.1 0>})
  n25 = Multiply(
   A=ComponentMask(None=TextureCoordinate()) {g=true}
   B=MaterialFunctionCall() {materialFunction=&ObjectScale}.Scale X
  )
  n26 = Multiply(A=Time(), B=@n22)
  n27 = Multiply(A=Sine(None=Add(A=@n25, B=@n26)) {period=4}, B=Constant3Vector() {constant=<rgba 0 2 0.5 0>})
  n28 = ComponentMask(None=TextureCoordinate()) {g=true}
  n29 = MaterialFunctionCall() {materialFunction=&ObjectScale}
  n30 = Time()
  n31 = Multiply(A=@n22, B=Constant() {r=2})
  n32 = Sine(None=Add(A=Multiply(A=@n28, B=@n29.Scale X), B=Multiply(A=@n30, B=@n31) {constB=2})) {period=6}
  n33 = Constant3Vector() {constant=<rgba 0 0 2 0>}
  n34 = VertexColor()
  n35 = Multiply(A=Add(A=Add(A=@n24, B=@n27), B=Multiply(A=@n32, B=@n33)), B=Multiply(A=@n34.R, B=OneMinus(None=@n34.G)))
  n36 = ScalarParameter() {defaultValue=1, parameterName=WiggleStrength}
  n37 = Sine(None=ComponentMask(None=TextureCoordinate()) {g=true}) {period=2}
  n38 = Constant() {r=0.5}
  n39 = ScalarParameter() {defaultValue=0.5472, parameterName=SphereRadius}
  n40 = ObjectRadius()
  n41 = ScalarParameter() {defaultValue=1000, parameterName=MaxLength}
  n42 = Clamp(None=LinearInterpolate(Alpha=Divide(A=@n40, B=@n41)))
  NamedRerouteDeclaration(None=Multiply(A=Multiply(A=@n35, B=@n36), B=SphereMask(A=@n37, B=@n38, Radius=@n39, Hardness=@n42) {attenuationRadius=0.5, hardnessPercent=50})) {name=MF_balloonRopeWiggle, nodeColor=<rgba 0 0.9062 1 1>}
  n43 = TextureSampleParameter2D() {parameterName=Diffuse, texture=/DealershipSharedLibrary/Assets/Sets/Ballon/Texture/T_Ballon_D}
  n44 = VectorParameter() {parameterName=bag plastic color, defaultValue=<rgba 0.2549 0.4004 0.5625 1>}
  n45 = TextureSample() {
   samplerSource=SSM_Wrap_WorldGroupSettings
   samplerType=SAMPLERTYPE_Masks
   texture=/DealershipSharedLibrary/Assets/Sets/Ballon/Texture/T_Ballon_FX_RGB
  }
  n46 = ScalarParameter() {defaultValue=-0.25, parameterName=rerer}
  n47 = ScalarParameter() {defaultValue=0.8, group=SS Shader Model, parameterName=gdsgfdsf}
  n48 = Fresnel(ExponentIn=ScalarParameter() {defaultValue=1, group=SS Shader Model, parameterName=Bag Fresnel Plastic Exponent})
  n49 = Multiply(A=@n45.B, B=Saturate(None=LinearInterpolate(A=@n46, B=@n47, Alpha=@n48)))
  n50 = ParticleColor()
  n51 = TextureSampleParameter2D() {parameterName=SpecularMasks, samplerType=SAMPLERTYPE_Masks, texture=/DealershipSharedLibrary/Assets/Sets/Ballon/Texture/T_Balloon_S}
  n52 = ScalarParameter() {defaultValue=0.05, group=SS Shader Model, parameterName=SS Inner Strength}
  n53 = ScalarParameter() {defaultValue=1.5, group=SS Shader Model, parameterName=SS Outer Strength}
  n54 = Fresnel(ExponentIn=ScalarParameter() {defaultValue=4, group=SS Shader Model, parameterName=SS Fresnel Exponent})
  n55 = Saturate(None=Multiply(A=@n50.RGB, B=ScalarParameter() {defaultValue=0.5, parameterName=balloon ss color modifier}))
  n56 = MakeMaterialAttributes!(
   BaseColor=LinearInterpolate(A=LinearInterpolate(A=@n43.RGB, B=@n44.RGB, Alpha=@n49), B=Multiply(A=@n43.RGB, B=@n50.RGB), Alpha=@n45.G)
   Metallic=@n51.G
   Specular=@n51.G
   Roughness=LinearInterpolate(A=LinearInterpolate(Alpha=@n51.B) {constA=0.1}, B=ScalarParameter() {defaultValue=0.1, parameterName=bag roughness}, Alpha=@n49)
   Anisotropy
   EmissiveColor
   Opacity=LinearInterpolate(B=LinearInterpolate(A=@n52, B=@n53, Alpha=@n54), Alpha=@n45.R) {constB=0}
   OpacityMask=ParticleColor().A
   Normal=TextureSampleParameter2D() {parameterName=Normals, samplerType=SAMPLERTYPE_Normal, texture=/DealershipSharedLibrary/Assets/Sets/Ballon/Texture/T_Balloon_N}.RGB
   Tangent
   WorldPositionOffset=LinearInterpolate(A=NamedRerouteUsage(), B=NamedRerouteUsage(), Alpha=ScalarParameter() {parameterName=UseCreepyBalloonMovement})
   SubsurfaceColor=LinearInterpolate(B=LinearInterpolate(A=@n55, B=@n50.RGB, Alpha=@n54), Alpha=@n45.R)
   ClearCoat
   ClearCoatRoughness
   AmbientOcclusion
   Refraction
   CustomizedUVs_0
   CustomizedUVs_1
   CustomizedUVs_2
   CustomizedUVs_3
   CustomizedUVs_4
   CustomizedUVs_5
   CustomizedUVs_6
   CustomizedUVs_7
   PixelDepthOffset
   ShadingModel
   Displacement
  )
  n57 = Constant()
  n58 = SetMaterialAttributes!(MaterialAttributes=@n56, Subsurface Color=@n57, Opacity=@n57) {attributeSetTypes [#1, #2]}
  n59 = FeatureLevelSwitch(Default=@n56, ES2=@n58, ES3_1=@n58, SM4_REMOVED=@n56)
  QualitySwitch(Default=@n59, Low=@n58, High=@n59, Medium=@n59)
 ]
}
